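# Physics asset for the male medium Historian body's loincloth, with capsule bodies for the cloth bones, belt, pelvis and thighs.
# Front and back panels each form a four-link center chain flanked by two-link side chains, with angular limits on each link.
# Back chains hang from the belt, front chains from separate MSTR bodies.
# Angularly free constraints join belt and thighs to pelvis, and cross-link front-left's tip to two center links.
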PhysicsAsset /BRCosmetics/Characters/Player/Male/Medium/Bodies/M_MED_Historian/Meshes/M_MED_Historian_Dyn_loincloth
bodies {
 dyn_belt [{shapeType=Capsule, length=0.505, radius=0.505}]
 dyn_loincloth_bk_01 [{shapeType=Capsule, length=4.4062, radius=10, center=<4.7651 0 10>, rotation=<rot 90 0 0>}]
 dyn_loincloth_bk_02 [{shapeType=Capsule, length=4.4062, radius=13.5, center=<4.7651 0 10>, rotation=<rot 90 0 0>}]
 dyn_loincloth_bk_03 [{shapeType=Capsule, length=4.4062, radius=13.5, center=<4.7651 0 10>, rotation=<rot 90 0 0>}]
 dyn_loincloth_bk_04 [{shapeType=Capsule, length=4.4062, radius=13.5, center=<4.7651 0 10>, rotation=<rot 90 0 0>}]
 dyn_loincloth_bk_l_01 [{shapeType=Capsule, length=5.4062, radius=9.5, center=<4.7651 0 10>, rotation=<rot 90 0 0>}]
 dyn_loincloth_bk_l_02 [{shapeType=Capsule, length=5.3438, radius=11, center=<4.7651 0 10>, rotation=<rot 90 0 0>}]
 dyn_loincloth_bk_r_01 [{shapeType=Capsule, length=5.4062, radius=9.5, center=<4.7652 0 8.6>, rotation=<rot 90 0 0>}]
 dyn_loincloth_bk_r_02 [{shapeType=Capsule, length=4.4062, radius=11, center=<4.7652 0 8.6>, rotation=<rot 90 0 0>}]
 dyn_loincloth_fr_01 [{shapeType=Capsule, length=9.375, radius=5, center=<4.7651 0 5>, rotation=<rot 90 0 0>}]
 dyn_loincloth_fr_02 [{shapeType=Capsule, length=9.375, radius=5, center=<4.7651 0 5>, rotation=<rot 90 0 0>}]
 dyn_loincloth_fr_03 [{shapeType=Capsule, length=9.375, radius=5, center=<4.7651 0 5>, rotation=<rot 90 0 0>}]
 dyn_loincloth_fr_04 [{shapeType=Capsule, length=9.375, radius=5, center=<4.7651 0 5>, rotation=<rot 90 0 0>}]
 dyn_loincloth_fr_MSTR [{shapeType=Capsule, length=0.505, radius=0.505}]
 dyn_loincloth_fr_l_01 [{shapeType=Capsule, length=8.7336, radius=4, shapeName=dyn_loincloth_fr_l_01_capsule, center=<6.1638 0.8873 -0.041>, rotation=<rot 90 0 0>}]
 dyn_loincloth_fr_l_02 [{shapeType=Capsule, length=7.7979, radius=4, shapeName=dyn_loincloth_fr_l_02_capsule, center=<14.152 2.3654 0.4655>, rotation=<rot 90 0 0>}]
 dyn_loincloth_fr_l_MSTR [{shapeType=Capsule, length=0.505, radius=0.505}]
 dyn_loincloth_fr_r_01 [{shapeType=Capsule, length=5.875, radius=8.5312, center=<4.765 0 7.1192>, rotation=<rot 90 0 0>}]
 dyn_loincloth_fr_r_02 [{shapeType=Capsule, length=13.5625, radius=7.8125, center=<6.9049 0.0004 4.9423>, rotation=<rot 87.4383 0.0001 0.0001>}]
 dyn_loincloth_fr_r_MSTR [{shapeType=Capsule, length=0.505, radius=0.505}]
 pelvis [{shapeType=Capsule, length=9.4163, radius=12.8125, center=<0 -1.8526 0.7152>, rotation=<rot 0 0 -0.9064>}]
 thigh_l [{shapeType=Capsule, length=19.4688, radius=12.505, center=<-22.4447 -2.0308 2.7228>, rotation=<rot 90 -90 -92.1981>}]
 thigh_r [{shapeType=Capsule, length=18.4688, radius=13.505, center=<23.7161 3.0948 -1.7609>, rotation=<rot 89.4539 -175.5101 -176.0004>}]
}
constraints [
 {bone1Name=dyn_loincloth_fr_l_01, bone2Name=dyn_loincloth_fr_l_MSTR, swing1LimitDegrees=10, swing2LimitDegrees=15, twistLimitDegrees=5}
 {bone1Name=dyn_loincloth_fr_l_02, bone2Name=dyn_loincloth_fr_l_01, swing1LimitDegrees=10, swing2LimitDegrees=15, twistLimitDegrees=5}
 {bone1Name=dyn_loincloth_bk_01, bone2Name=dyn_belt, swing1LimitDegrees=15, swing2LimitDegrees=15, twistLimitDegrees=5}
 {bone1Name=dyn_loincloth_bk_02, bone2Name=dyn_loincloth_bk_01, swing1LimitDegrees=20, swing2LimitDegrees=45, twistLimitDegrees=5}
 {bone1Name=dyn_loincloth_bk_03, bone2Name=dyn_loincloth_bk_02, swing1LimitDegrees=30, swing2LimitDegrees=30, twistLimitDegrees=5}
 {bone1Name=dyn_loincloth_bk_04, bone2Name=dyn_loincloth_bk_03, swing1LimitDegrees=45, swing2LimitDegrees=45, twistLimitDegrees=5}
 {bone1Name=dyn_loincloth_bk_l_01, bone2Name=dyn_belt, swing1LimitDegrees=45, swing2LimitDegrees=45, twistLimitDegrees=5}
 {bone1Name=dyn_loincloth_bk_l_02, bone2Name=dyn_loincloth_bk_l_01, swing1LimitDegrees=45, swing2LimitDegrees=45, twistLimitDegrees=5}
 {bone1Name=dyn_loincloth_bk_r_01, bone2Name=dyn_belt, swing1LimitDegrees=45, swing2LimitDegrees=45, twistLimitDegrees=5}
 {bone1Name=dyn_loincloth_bk_r_02, bone2Name=dyn_loincloth_bk_r_01, swing1LimitDegrees=45, swing2LimitDegrees=45, twistLimitDegrees=5}
 {bone1Name=dyn_loincloth_fr_01, bone2Name=dyn_loincloth_fr_MSTR, swing1LimitDegrees=15, swing2LimitDegrees=90, twistLimitDegrees=5}
 {bone1Name=dyn_loincloth_fr_02, bone2Name=dyn_loincloth_fr_01, swing1LimitDegrees=45, swing2LimitDegrees=45, twistLimitDegrees=5}
 {bone1Name=dyn_loincloth_fr_03, bone2Name=dyn_loincloth_fr_02, swing1LimitDegrees=45, swing2LimitDegrees=45, twistLimitDegrees=5}
 {bone1Name=dyn_loincloth_fr_04, bone2Name=dyn_loincloth_fr_03, swing1LimitDegrees=45, swing2LimitDegrees=45, twistLimitDegrees=5}
 {bone1Name=dyn_loincloth_fr_r_01, bone2Name=dyn_loincloth_fr_r_MSTR, swing1LimitDegrees=15, swing2LimitDegrees=15, twistLimitDegrees=5}
 {bone1Name=dyn_loincloth_fr_r_02, bone2Name=dyn_loincloth_fr_r_01, swing1LimitDegrees=15, swing2LimitDegrees=15, twistLimitDegrees=5}
 {
  bone1Name=dyn_belt
  bone2Name=pelvis
  swing1LimitDegrees=45
  swing1Motion=Free
  swing2LimitDegrees=45
  swing2Motion=Free
  twistLimitDegrees=45
  twistMotion=Free
 }
 {
  bone1Name=thigh_l
  bone2Name=pelvis
  swing1LimitDegrees=45
  swing1Motion=Free
  swing2LimitDegrees=45
  swing2Motion=Free
  twistLimitDegrees=45
  twistMotion=Free
 }
 {
  bone1Name=thigh_r
  bone2Name=pelvis
  swing1LimitDegrees=45
  swing1Motion=Free
  swing2LimitDegrees=45
  swing2Motion=Free
  twistLimitDegrees=45
  twistMotion=Free
 }
 {
  bone1Name=dyn_loincloth_fr_03
  bone2Name=dyn_loincloth_fr_l_02
  swing1LimitDegrees=45
  swing1Motion=Free
  swing2LimitDegrees=45
  swing2Motion=Free
  twistLimitDegrees=45
  twistMotion=Free
 }
 {
  bone1Name=dyn_loincloth_fr_02
  bone2Name=dyn_loincloth_fr_l_02
  swing1LimitDegrees=45
  swing1Motion=Free
  swing2LimitDegrees=45
  swing2Motion=Free
  twistLimitDegrees=45
  twistMotion=Free
 }
]
