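# Physics asset for the dynamic torso bones of the male medium Historian cosmetic body, covering belt-hung gear and cloth.
# Capsule bodies chain from the pelvis through the belt to bag, horn, knife, satchel and knot-strap chains, with loincloth chains off the knot master.
# Tiny master-bone capsules act as attachment points with free angular motion, as do loincloth links and fur roots.
# Gear and strap links carry tighter swing and twist limits.
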PhysicsAsset /BRCosmetics/Characters/Player/Male/Medium/Bodies/M_MED_Historian/Meshes/M_MED_Historian_Dyn_torso
bodies {
 dyn_back_fur_01_MSTR [{shapeType=Capsule, length=0.505, radius=0.505}]
 dyn_bag_l_01 [{shapeType=Capsule, length=2, radius=5, center=<2.6494 0 5>, rotation=<rot 90 0 0>}]
 dyn_bag_l_02 [{shapeType=Capsule, length=3.25, radius=5.875, center=<7.3031 0 5>, rotation=<rot 90 0 0>}]
 dyn_bag_l_MSTR [{shapeType=Capsule, length=0.505, radius=0.505}]
 dyn_belt [{shapeType=Capsule, length=0.505, radius=0.505}]
 dyn_belt_knot [{shapeType=Capsule, length=1.0625, radius=1.375, center=<1.2015 0 -0.0003>, rotation=<rot 90 0 0>}]
 dyn_belt_knot_MSTR [{shapeType=Capsule, length=0.505, radius=0.505}]
 dyn_belt_strap_fr_01 [{shapeType=Capsule, length=2.1875, radius=4.1875, center=<7.544 0.0001 3.3838>, rotation=<rot 90 0 0>}]
 dyn_belt_strap_fr_02 [{shapeType=Capsule, length=5.875, radius=4.1875, center=<4.8583 0 3.3838>, rotation=<rot 90 0 0>}]
 dyn_belt_strap_fr_03 [{shapeType=Capsule, length=5.875, radius=4.1875, center=<4.8583 0 3.3838>, rotation=<rot 90 0 0>}]
 dyn_horn_r_01 [
  {shapeType=Capsule, length=4.9688, radius=6.125, center=<4.8583 0 -0.0004>, rotation=<rot 90 0 0>}
  {shapeType=Capsule, length=23.9375, radius=3.5312, center=<9.4974 1.7069 1.8819>, rotation=<rot 81.6612 54.9917 0>}
 ]
 dyn_horn_r_MSTR [{shapeType=Capsule, length=0.505, radius=0.505}]
 dyn_knife_l_01 [{shapeType=Capsule, length=4.7188, radius=4.1562, center=<4.2831 0 -0.0003>, rotation=<rot 90 0 0>}]
 dyn_knife_l_02 [{shapeType=Capsule, length=30.5625, radius=2.375, center=<2.4632 0.2142 1.8499>, rotation=<rot 88.2337 -20.0033 2.3902>}]
 dyn_knife_l_MSTR [{shapeType=Capsule, length=0.505, radius=0.505}]
 dyn_loincloth_fr_01 [{shapeType=Capsule, length=27.25, radius=14.3125, center=<10.5236 0.0005 -13.5789>, rotation=<rot 90 0 0>}]
 dyn_loincloth_fr_02 [{shapeType=Capsule, length=27.25, radius=14.3125, center=<10.5236 0.0005 -13.5789>, rotation=<rot 90 0 0>}]
 dyn_loincloth_fr_03 [{shapeType=Capsule, length=27.25, radius=14.3125, center=<10.5236 0.0005 -13.5789>, rotation=<rot 90 0 0>}]
 dyn_loincloth_fr_r_01 [{shapeType=Capsule, length=18.4375, radius=23.125, center=<10.5236 -0.0001 -20.7453>, rotation=<rot 90 0 0>}]
 dyn_loincloth_fr_r_02 [{shapeType=Capsule, length=18.4375, radius=23.125, center=<10.5236 -0.0001 -20.7453>, rotation=<rot 90 0 0>}]
 dyn_omega_01_MSTR [{shapeType=Capsule, length=0.505, radius=0.505}]
 dyn_satchel_l_01 [{shapeType=Capsule, length=1.2188, radius=6.125, center=<3.5043 0 -0.0004>, rotation=<rot 90 0 0>}]
 dyn_satchel_l_02 [{shapeType=Capsule, length=12.4375, radius=5.5625, center=<3.014 0 -0.0002>, rotation=<rot 0 0 89.9995>}]
 dyn_satchel_l_MSTR [{shapeType=Capsule, length=0.505, radius=0.505}]
 dyn_side_fur_01_MSTR [{shapeType=Capsule, length=0.505, radius=0.505}]
 pelvis [{shapeType=Capsule, length=9.875, radius=14.6875, center=<-4.8042 -0.0001 0>}]
}
constraints [
 {
  bone1Name=dyn_bag_l_MSTR
  bone2Name=dyn_belt
  swing1LimitDegrees=45
  swing1Motion=Free
  swing2LimitDegrees=45
  swing2Motion=Free
  twistLimitDegrees=45
  twistMotion=Free
 }
 {bone1Name=dyn_bag_l_01, bone2Name=dyn_bag_l_MSTR, swing1LimitDegrees=30, swing2LimitDegrees=30, twistLimitDegrees=5}
 {bone1Name=dyn_bag_l_02, bone2Name=dyn_bag_l_01, swing1LimitDegrees=30, swing2LimitDegrees=30, twistLimitDegrees=5}
 {
  bone1Name=dyn_belt_knot_MSTR
  bone2Name=dyn_belt
  swing1LimitDegrees=45
  swing1Motion=Free
  swing2LimitDegrees=45
  swing2Motion=Free
  twistLimitDegrees=45
  twistMotion=Free
 }
 {bone1Name=dyn_belt_knot, bone2Name=dyn_belt_knot_MSTR, swing1LimitDegrees=30, swing2LimitDegrees=30, twistLimitDegrees=5}
 {bone1Name=dyn_belt_strap_fr_01, bone2Name=dyn_belt_knot, swing1LimitDegrees=30, swing2LimitDegrees=15, twistLimitDegrees=15}
 {bone1Name=dyn_belt_strap_fr_02, bone2Name=dyn_belt_strap_fr_01, swing1LimitDegrees=30, swing2LimitDegrees=15, twistLimitDegrees=15}
 {bone1Name=dyn_belt_strap_fr_03, bone2Name=dyn_belt_strap_fr_02, swing1LimitDegrees=30, swing2LimitDegrees=15, twistLimitDegrees=15}
 {
  bone1Name=dyn_horn_r_MSTR
  bone2Name=dyn_belt
  swing1LimitDegrees=45
  swing1Motion=Free
  swing2LimitDegrees=45
  swing2Motion=Free
  twistLimitDegrees=45
  twistMotion=Free
 }
 {bone1Name=dyn_horn_r_01, bone2Name=dyn_horn_r_MSTR, swing1LimitDegrees=30, swing2LimitDegrees=30, twistLimitDegrees=5}
 {
  bone1Name=dyn_knife_l_MSTR
  bone2Name=dyn_belt
  swing1LimitDegrees=45
  swing1Motion=Free
  swing2LimitDegrees=45
  swing2Motion=Free
  twistLimitDegrees=45
  twistMotion=Free
 }
 {bone1Name=dyn_knife_l_01, bone2Name=dyn_knife_l_MSTR, swing1LimitDegrees=10, swing2LimitDegrees=30, twistLimitDegrees=5}
 {bone1Name=dyn_knife_l_02, bone2Name=dyn_knife_l_01, swing1LimitDegrees=5, swing2LimitDegrees=5, twistLimitDegrees=5}
 {
  bone1Name=dyn_satchel_l_MSTR
  bone2Name=dyn_belt
  swing1LimitDegrees=45
  swing1Motion=Free
  swing2LimitDegrees=45
  swing2Motion=Free
  twistLimitDegrees=45
  twistMotion=Free
 }
 {bone1Name=dyn_satchel_l_01, bone2Name=dyn_satchel_l_MSTR, swing1LimitDegrees=30, swing2LimitDegrees=30, twistLimitDegrees=5}
 {bone1Name=dyn_satchel_l_02, bone2Name=dyn_satchel_l_01, swing1LimitDegrees=30, swing2LimitDegrees=30, twistLimitDegrees=5}
 {
  bone1Name=dyn_loincloth_fr_03
  bone2Name=dyn_loincloth_fr_02
  swing1LimitDegrees=45
  swing1Motion=Free
  swing2LimitDegrees=45
  swing2Motion=Free
  twistLimitDegrees=45
  twistMotion=Free
 }
 {
  bone1Name=dyn_loincloth_fr_02
  bone2Name=dyn_loincloth_fr_01
  swing1LimitDegrees=45
  swing1Motion=Free
  swing2LimitDegrees=45
  swing2Motion=Free
  twistLimitDegrees=45
  twistMotion=Free
 }
 {
  bone1Name=dyn_loincloth_fr_01
  bone2Name=dyn_belt_knot_MSTR
  swing1LimitDegrees=45
  swing1Motion=Free
  swing2LimitDegrees=45
  swing2Motion=Free
  twistLimitDegrees=45
  twistMotion=Free
 }
 {
  bone1Name=dyn_loincloth_fr_r_02
  bone2Name=dyn_loincloth_fr_r_01
  swing1LimitDegrees=45
  swing1Motion=Free
  swing2LimitDegrees=45
  swing2Motion=Free
  twistLimitDegrees=45
  twistMotion=Free
 }
 {
  bone1Name=dyn_loincloth_fr_r_01
  bone2Name=dyn_belt_knot_MSTR
  swing1LimitDegrees=45
  swing1Motion=Free
  swing2LimitDegrees=45
  swing2Motion=Free
  twistLimitDegrees=45
  twistMotion=Free
 }
 {
  bone1Name=dyn_omega_01_MSTR
  bone2Name=pelvis
  swing1LimitDegrees=45
  swing1Motion=Free
  swing2LimitDegrees=45
  swing2Motion=Free
  twistLimitDegrees=45
  twistMotion=Free
 }
 {
  bone1Name=dyn_back_fur_01_MSTR
  bone2Name=pelvis
  swing1LimitDegrees=45
  swing1Motion=Free
  swing2LimitDegrees=45
  swing2Motion=Free
  twistLimitDegrees=45
  twistMotion=Free
 }
 {
  bone1Name=dyn_side_fur_01_MSTR
  bone2Name=pelvis
  swing1LimitDegrees=45
  swing1Motion=Free
  swing2LimitDegrees=45
  swing2Motion=Free
  twistLimitDegrees=45
  twistMotion=Free
 }
 {
  bone1Name=dyn_belt
  bone2Name=pelvis
  swing1LimitDegrees=45
  swing1Motion=Free
  swing2LimitDegrees=45
  swing2Motion=Free
  twistLimitDegrees=45
  twistMotion=Free
 }
]
