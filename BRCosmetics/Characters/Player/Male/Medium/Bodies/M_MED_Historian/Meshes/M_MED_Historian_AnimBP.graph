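# Secondary-motion AnimBlueprint for the male medium Historian body cosmetic, copying the parent mesh's pose and layering procedural deformation and physics on it.
# Deform Control Rigs and weighted bone constraints feed a loincloth RigidBody, AnimDynamics for fur, horn and hanging gear, then a strips rig and RigidBody.
# RigidBody sims run relative to dyn_root_simspace, rotated by a computed root-space rotation.
# Skydive, emote and front-end variables adjust selected blends.
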AnimBlueprint /BRCosmetics/Characters/Player/Male/Medium/Bodies/M_MED_Historian/Meshes/M_MED_Historian_AnimBP
target_skeleton=&/M_MED_Historian_Skeleton
graphs [
 AnimGraph {nodes [
   n1 = Get GravityOverride()
   RigidBody!(ComponentPose, Alpha, OverrideWorldGravity:Vector=@n1, ExternalForce) -> (Pose) {
    bClampLinearTranslationLimitToRefPose=true
    bOverrideWorldGravity=true
    overridePhysicsAsset=&/M_MED_Historian_Dyn_torso
    simulationSpace=BaseBoneSpace
    alphaBoolBlend {blendOption=Linear}
    baseBoneRef {boneName=dyn_root_simspace}
    componentAppliedLinearAccClamp=<3000 3000 3000>
    componentLinearAccScale {z=0.5}
    componentLinearVelScale {z=0.5}
    overrideWorldGravity {z=-980}
    simSpaceSettings {
     maxAngularAcceleration=1000
     maxAngularVelocity=3
     maxLinearAcceleration=250
     maxLinearVelocity=500
     velocityScaleZ=0
     worldAlpha=0.5
     externalLinearDragV=<0.5 0.5>
    }
   }
   n2 = Control Rig(Source=Copy Pose From Mesh() {bUseAttachedParent=true}) {
    alphaBoolBlend {blendOption=Linear}
    controlRigAssetReference {blueprintRigClass=&/M_MED_Historian_Deform_CtrlRig_twists.M_MED_Historian_Deform_CtrlRig_twists_C}
   }
   n3 = Control Rig(Source=@n2) {
    alphaBoolBlend {blendOption=Linear}
    controlRigAssetReference {blueprintRigClass=&/M_MED_Historian_Deform_CtrlRig_sculpters.M_MED_Historian_Deform_CtrlRig_sculpters_C}
   }
   n4 = Control Rig(Source=@n3) {
    alphaBoolBlend {blendOption=Linear}
    controlRigAssetReference {blueprintRigClass=&/M_MED_Historian_Deform_CtrlRig_deformers.M_MED_Historian_Deform_CtrlRig_deformers_C}
   }
   n5 = "Constraint - dyn_side_fur_01_MSTR (PPPPPP)"!(
    ComponentPose=Local To Component(LocalPose=@n4)
    Alpha=1
    ConstraintWeights_0=0.0456
    ConstraintWeights_1:float=0.1026
    ConstraintWeights_2:float=0.3084
    ConstraintWeights_3:float=0.3431
    ConstraintWeights_4:float=0.1708
    ConstraintWeights_5:float=0.0295
   ) -> (Pose) {
    alphaBoolBlend {blendOption=Linear}
    boneToModify {boneName=dyn_side_fur_01_MSTR}
    constraintSetup [
     {targetBone {boneName=dyn_side_fur_01}}
     {targetBone {boneName=spine_01}}
     {targetBone {boneName=spine_02}}
     {targetBone {boneName=spine_03}}
     {targetBone {boneName=spine_04}}
     {targetBone {boneName=pelvis}}
    ]
    constraintWeights [1,1,1,1,1,1]
   }
   n6 = "Constraint - dyn_back_fur_01_MSTR (PPP)"!(ComponentPose=@n5, Alpha=1, ConstraintWeights_0=0.23, ConstraintWeights_1:float=0.3702, ConstraintWeights_2:float=0.3997) -> (Pose) {
    alphaBoolBlend {blendOption=Linear}
    boneToModify {boneName=dyn_back_fur_01_MSTR}
    constraintSetup [{targetBone {boneName=spine_03}}, {targetBone {boneName=spine_04}}, {targetBone {boneName=spine_05}}]
    constraintWeights [1,1,1]
   }
   n7 = "Constraint - dyn_omega_01_MSTR (PPPP)"!(
    ComponentPose=@n6
    Alpha=1
    ConstraintWeights_0=0.2683
    ConstraintWeights_1:float=0.1665
    ConstraintWeights_2:float=0.1775
    ConstraintWeights_3:float=0.3877
   ) -> (Pose) {
    alphaBoolBlend {blendOption=Linear}
    boneToModify {boneName=dyn_omega_01_MSTR}
    constraintSetup [{targetBone {boneName=clavicle_r}}, {targetBone {boneName=spine_03}}, {targetBone {boneName=spine_04}}, {targetBone {boneName=spine_05}}]
    constraintWeights [1,1,1,1]
   }
   n8 = "Constraint - dyn_horn_r_MSTR (PPPPP)"!(
    ComponentPose=@n7
    Alpha=1
    ConstraintWeights_0=0.2165
    ConstraintWeights_1:float=0.0271
    ConstraintWeights_2:float=0.011
    ConstraintWeights_3:float=0.0204
    ConstraintWeights_4:float=0.7251
   ) -> (Pose) {
    alphaBoolBlend {blendOption=Linear}
    boneToModify {boneName=dyn_horn_r_MSTR}
    constraintSetup [
     {targetBone {boneName=spine_01}}
     {targetBone {boneName=deform_thigh_bk_r}}
     {targetBone {boneName=spine_03}}
     {targetBone {boneName=spine_02}}
     {targetBone {boneName=pelvis}}
    ]
    constraintWeights [1,1,1,1,1]
   }
   n9 = "Constraint - dyn_satchel_l_MSTR (PPPP)"!(
    ComponentPose=@n8
    Alpha=1
    ConstraintWeights_0=0.2733
    ConstraintWeights_1:float=0.042
    ConstraintWeights_2:float=0.0328
    ConstraintWeights_3:float=0.6519
   ) -> (Pose) {
    alphaBoolBlend {blendOption=Linear}
    boneToModify {boneName=dyn_satchel_l_MSTR}
    constraintSetup [{targetBone {boneName=spine_01}}, {targetBone {boneName=spine_02}}, {targetBone {boneName=spine_03}}, {targetBone {boneName=pelvis}}]
    constraintWeights [1,1,1,1]
   }
   n10 = "Constraint - dyn_knife_l_MSTR (PPPP)"!(
    ComponentPose=@n9
    Alpha=1
    ConstraintWeights_0=0.3514
    ConstraintWeights_1:float=0.0892
    ConstraintWeights_2:float=0.0284
    ConstraintWeights_3:float=0.5311
   ) -> (Pose) {
    alphaBoolBlend {blendOption=Linear}
    boneToModify {boneName=dyn_knife_l_MSTR}
    constraintSetup [{targetBone {boneName=spine_01}}, {targetBone {boneName=spine_02}}, {targetBone {boneName=spine_03}}, {targetBone {boneName=pelvis}}]
    constraintWeights [1,1,1,1]
   }
   n11 = "Constraint - dyn_bag_l_MSTR (PPPPP)"!(
    ComponentPose=@n10
    Alpha=1
    ConstraintWeights_0=0.5397
    ConstraintWeights_1:float=0.012
    ConstraintWeights_2:float=0.0742
    ConstraintWeights_3:float=0.026
    ConstraintWeights_4:float=0.3482
   ) -> (Pose) {
    alphaBoolBlend {blendOption=Linear}
    boneToModify {boneName=dyn_bag_l_MSTR}
    constraintSetup [
     {targetBone {boneName=pelvis}}
     {targetBone {boneName=deform_thigh_fr_l}}
     {targetBone {boneName=spine_02}}
     {targetBone {boneName=spine_03}}
     {targetBone {boneName=spine_01}}
    ]
    constraintWeights [1,1,1,1,1]
   }
   n12 = "Constraint - dyn_buckle_fr_l (PPPPP)"!(
    ComponentPose=@n11
    Alpha=1
    ConstraintWeights_0=0.0703
    ConstraintWeights_1:float=0.0278
    ConstraintWeights_2:float=0.1771
    ConstraintWeights_3:float=0.5491
    ConstraintWeights_4:float=0.1757
   ) -> (Pose) {
    alphaBoolBlend {blendOption=Linear}
    boneToModify {boneName=dyn_buckle_fr_l}
    constraintSetup [
     {targetBone {boneName=deform_pectoral_fr_l}}
     {targetBone {boneName=spine_02}}
     {targetBone {boneName=spine_03}}
     {targetBone {boneName=spine_04}}
     {targetBone {boneName=spine_05}}
    ]
    constraintWeights [1,1,1,1,1]
   }
   n13 = "Constraint - dyn_strip_r_A_MSTR (PPPPP)"!(
    ComponentPose=@n12
    Alpha=1
    ConstraintWeights_0=0.1517
    ConstraintWeights_1:float=0.065
    ConstraintWeights_2:float=0.1399
    ConstraintWeights_3:float=0.3905
    ConstraintWeights_4:float=0.2529
   ) -> (Pose) {
    alphaBoolBlend {blendOption=Linear}
    boneToModify {boneName=dyn_strip_r_A_MSTR}
    constraintSetup [
     {targetBone {boneName=thigh_twist_01_r}}
     {targetBone {boneName=spine_01}}
     {targetBone {boneName=dyn_loincloth_fr_r_01}}
     {targetBone {boneName=pelvis}}
     {targetBone {boneName=deform_thigh_fr_r}}
    ]
    constraintWeights [1,1,1,1,1]
   }
   n14 = "Constraint - dyn_strip_r_B_MSTR (PPP)"!(ComponentPose=@n13, Alpha=1, ConstraintWeights_0=0.5999, ConstraintWeights_1:float=0.3788, ConstraintWeights_2:float=0.0214) -> (Pose) {
    alphaBoolBlend {blendOption=Linear}
    boneToModify {boneName=dyn_strip_r_B_MSTR}
    constraintSetup [{targetBone {boneName=pelvis}}, {targetBone {boneName=dyn_belt_knot_MSTR}}, {targetBone {boneName=spine_01}}]
    constraintWeights [1,1,1]
   }
   n15 = "Constraint - dyn_strip_l_H_MSTR (PPPPPP)"!(
    ComponentPose=@n14
    Alpha=1
    ConstraintWeights_0=0.4261
    ConstraintWeights_1:float=0.1999
    ConstraintWeights_2:float=0.0107
    ConstraintWeights_3:float=0.3026
    ConstraintWeights_4:float=0.0172
    ConstraintWeights_5:float=0.0435
   ) -> (Pose) {
    alphaBoolBlend {blendOption=Linear}
    boneToModify {boneName=dyn_strip_l_H_MSTR}
    constraintSetup [
     {targetBone {boneName=dyn_hip_deformation_l}}
     {targetBone {boneName=dyn_strip_l_H_MSTR}}
     {targetBone {boneName=spine_01}}
     {targetBone {boneName=pelvis}}
     {targetBone {boneName=dyn_strip_l_F_MSTR}}
     {targetBone {boneName=dyn_strip_l_J_MSTR}}
    ]
    constraintWeights [1,1,1,1,1,1]
   }
   n16 = "Constraint - dyn_strip_l_J_MSTR (PPPPP)"!(
    ComponentPose=@n15
    Alpha=1
    ConstraintWeights_0=0.3791
    ConstraintWeights_1:float=0.3753
    ConstraintWeights_2:float=0.0605
    ConstraintWeights_3:float=0.1483
    ConstraintWeights_4:float=0.0368
   ) -> (Pose) {
    alphaBoolBlend {blendOption=Linear}
    boneToModify {boneName=dyn_strip_l_J_MSTR}
    constraintSetup [
     {targetBone {boneName=pelvis}}
     {targetBone {boneName=dyn_strip_l_J_MSTR}}
     {targetBone {boneName=dyn_hip_deformation_l}}
     {targetBone {boneName=dyn_strip_l_H_MSTR}}
     {targetBone {boneName=spine_01}}
    ]
    constraintWeights [1,1,1,1,1]
   }
   n17 = "Transform (Modify) Bone - Bone: dyn_root_simspace"(ComponentPose=@n16, Alpha=1, Translation="0,0,0", Rotation=Get root_space_rotation() -> (root_space_rotation:Rotator), Scale="1,1,1") {
    rotationMode=BMM_Replace
    rotationSpace=BCS_WorldSpace
    translationSpace=BCS_BoneSpace
    alphaBoolBlend {blendOption=Linear}
    alphaScaleBiasClamp {bInterpResult=true, interpSpeedDecreasing=2, interpSpeedIncreasing=2}
    boneToModify {boneName=dyn_root_simspace}
   }
   n18 = RigidBody(ComponentPose=@n17, Alpha=Get Skydive Loincloth Alpha() -> (Skydive Loincloth Alpha:double), OverrideWorldGravity=Get GravityOverride()) {
    bClampLinearTranslationLimitToRefPose=true
    bOverrideWorldGravity=true
    overridePhysicsAsset=&/M_MED_Historian_Dyn_loincloth
    simulationSpace=BaseBoneSpace
    alphaBoolBlend {blendOption=Linear}
    baseBoneRef {boneName=dyn_root_simspace}
    componentAppliedLinearAccClamp=<3000 3000 3000>
    componentLinearAccScale {z=0.5}
    componentLinearVelScale {z=0.5}
    overrideWorldGravity {z=-980}
    simSpaceSettings {
     maxAngularAcceleration=1000
     maxAngularVelocity=3
     maxLinearAcceleration=250
     maxLinearVelocity=500
     velocityScaleZ=0
     worldAlpha=1
     externalLinearDragV=<0.5 0.5>
    }
   }
   n19 = "Constraint - dyn_strip_l_A_MSTR (PPPPPP)"!(
    ComponentPose=@n18
    Alpha=1
    ConstraintWeights_0=0.4007
    ConstraintWeights_1:float=0.4973
    ConstraintWeights_2:float=0.0122
    ConstraintWeights_3:float=0.0355
    ConstraintWeights_4:float=0.0377
    ConstraintWeights_5:float=0.0166
   ) -> (Pose) {
    alphaBoolBlend {blendOption=Linear}
    boneToModify {boneName=dyn_strip_l_A_MSTR}
    constraintSetup [
     {targetBone {boneName=dyn_loincloth_fr_l_02}}
     {targetBone {boneName=dyn_loincloth_fr_l_01}}
     {targetBone {boneName=dyn_loincloth_fr_03}}
     {targetBone {boneName=dyn_loincloth_fr_02}}
     {targetBone {boneName=dyn_loincloth_fr_01}}
     {targetBone {boneName=pelvis}}
    ]
    constraintWeights [1,1,1,1,1,1]
   }
   n20 = "Constraint - dyn_strip_l_B_MSTR (PPPPPPP)"!(
    ComponentPose=@n19
    Alpha=1
    ConstraintWeights_0=0.2742
    ConstraintWeights_1:float=0.0487
    ConstraintWeights_2:float=0.5819
    ConstraintWeights_3:float=0.0195
    ConstraintWeights_4:float=0.023
    ConstraintWeights_5:float=0.0322
    ConstraintWeights_6:float=0.0206
   ) -> (Pose) {
    alphaBoolBlend {blendOption=Linear}
    boneToModify {boneName=dyn_strip_l_B_MSTR}
    constraintSetup [
     {targetBone {boneName=dyn_loincloth_fr_l_02}}
     {targetBone {boneName=dyn_loincloth_fr_l_MSTR}}
     {targetBone {boneName=dyn_loincloth_fr_l_01}}
     {targetBone {boneName=dyn_strip_l_D_MSTR}}
     {targetBone {boneName=dyn_loincloth_fr_02}}
     {targetBone {boneName=dyn_loincloth_fr_01}}
     {targetBone {boneName=pelvis}}
    ]
    constraintWeights [1,1,1,1,1,1,1]
   }
   n21 = "Constraint - dyn_strip_l_C_MSTR (PPPPPPPP)"!(
    ComponentPose=@n20
    Alpha=1
    ConstraintWeights_0=0.161
    ConstraintWeights_1:float=0.2096
    ConstraintWeights_2:float=0.4475
    ConstraintWeights_3:float=0.0677
    ConstraintWeights_4:float=0.0146
    ConstraintWeights_5:float=0.0279
    ConstraintWeights_6:float=0.0502
    ConstraintWeights_7:float=0.0215
   ) -> (Pose) {
    alphaBoolBlend {blendOption=Linear}
    boneToModify {boneName=dyn_strip_l_C_MSTR}
    constraintSetup [
     {targetBone {boneName=dyn_loincloth_fr_l_02}}
     {targetBone {boneName=dyn_loincloth_fr_l_MSTR}}
     {targetBone {boneName=dyn_loincloth_fr_l_01}}
     {targetBone {boneName=dyn_strip_l_D_MSTR}}
     {targetBone {boneName=dyn_loincloth_fr_02}}
     {targetBone {boneName=dyn_loincloth_fr_01}}
     {targetBone {boneName=dyn_hip_deformation_fr_l}}
     {targetBone {boneName=pelvis}}
    ]
    constraintWeights [1,1,1,1,1,1,1,1]
   }
   n22 = "Constraint - dyn_strip_l_K_MSTR (PPPPPPPPPP)"!(
    ComponentPose=@n21
    Alpha=1
    ConstraintWeights_0=0.0333
    ConstraintWeights_1:float=0.2243
    ConstraintWeights_2:float=0.2051
    ConstraintWeights_3:float=0.0336
    ConstraintWeights_4:float=0.0513
    ConstraintWeights_5:float=0.031
    ConstraintWeights_6:float=0.0504
    ConstraintWeights_7:float=0.0323
    ConstraintWeights_8:float=0.3247
    ConstraintWeights_9:float=0.014
   ) -> (Pose) {
    alphaBoolBlend {blendOption=Linear}
    boneToModify {boneName=dyn_strip_l_K_MSTR}
    constraintSetup [
     {targetBone {boneName=dyn_loincloth_bk_r_02}}
     {targetBone {boneName=dyn_loincloth_bk_l_01}}
     {targetBone {boneName=dyn_loincloth_bk_l_02}}
     {targetBone {boneName=dyn_loincloth_bk_04}}
     {targetBone {boneName=spine_01}}
     {targetBone {boneName=dyn_loincloth_bk_01}}
     {targetBone {boneName=dyn_loincloth_bk_02}}
     {targetBone {boneName=dyn_loincloth_bk_03}}
     {targetBone {boneName=pelvis}}
     {targetBone {boneName=deform_thigh_bk_l}}
    ]
    constraintWeights [1,1,1,1,1,1,1,1,1,1]
   }
   n23 = "Constraint - dyn_strip_l_L_MSTR (PPPPPPPP)"!(
    ComponentPose=@n22
    Alpha=1
    ConstraintWeights_0=0.0485
    ConstraintWeights_1:float=0.064
    ConstraintWeights_2:float=0.4551
    ConstraintWeights_3:float=0.1447
    ConstraintWeights_4:float=0.0149
    ConstraintWeights_5:float=0.1065
    ConstraintWeights_6:float=0.1115
    ConstraintWeights_7:float=0.0548
   ) -> (Pose) {
    alphaBoolBlend {blendOption=Linear}
    boneToModify {boneName=dyn_strip_l_L_MSTR}
    constraintSetup [
     {targetBone {boneName=dyn_loincloth_bk_r_02}}
     {targetBone {boneName=dyn_loincloth_bk_l_01}}
     {targetBone {boneName=dyn_loincloth_bk_l_02}}
     {targetBone {boneName=dyn_loincloth_bk_04}}
     {targetBone {boneName=dyn_loincloth_bk_01}}
     {targetBone {boneName=dyn_loincloth_bk_02}}
     {targetBone {boneName=dyn_loincloth_bk_03}}
     {targetBone {boneName=pelvis}}
    ]
    constraintWeights [1,1,1,1,1,1,1,1]
   }
   n24 = "Constraint - dyn_strip_l_M_MSTR (PPPPPPP)"!(
    ComponentPose=@n23
    Alpha=1
    ConstraintWeights_0=0.0569
    ConstraintWeights_1:float=0.0156
    ConstraintWeights_2:float=0.319
    ConstraintWeights_3:float=0.3136
    ConstraintWeights_4:float=0.1164
    ConstraintWeights_5:float=0.1601
    ConstraintWeights_6:float=0.0184
   ) -> (Pose) {
    alphaBoolBlend {blendOption=Linear}
    boneToModify {boneName=dyn_strip_l_M_MSTR}
    constraintSetup [
     {targetBone {boneName=dyn_loincloth_bk_r_02}}
     {targetBone {boneName=dyn_loincloth_bk_l_01}}
     {targetBone {boneName=dyn_loincloth_bk_l_02}}
     {targetBone {boneName=dyn_loincloth_bk_04}}
     {targetBone {boneName=dyn_loincloth_bk_02}}
     {targetBone {boneName=dyn_loincloth_bk_03}}
     {targetBone {boneName=pelvis}}
    ]
    constraintWeights [1,1,1,1,1,1,1]
   }
   n25 = "Constraint - dyn_strip_r_C_MSTR (PPPPPP)"!(
    ComponentPose=@n24
    Alpha=1
    ConstraintWeights_0=0.0363
    ConstraintWeights_1:float=0.226
    ConstraintWeights_2:float=0.0906
    ConstraintWeights_3:float=0.0429
    ConstraintWeights_4:float=0.3579
    ConstraintWeights_5:float=0.2464
   ) -> (Pose) {
    alphaBoolBlend {blendOption=Linear}
    boneToModify {boneName=dyn_strip_r_C_MSTR}
    constraintSetup [
     {targetBone {boneName=deform_thigh_bk_r}}
     {targetBone {boneName=thigh_twist_01_r}}
     {targetBone {boneName=dyn_loincloth_fr_r_MSTR}}
     {targetBone {boneName=spine_01}}
     {targetBone {boneName=pelvis}}
     {targetBone {boneName=deform_thigh_fr_r}}
    ]
    constraintWeights [1,1,1,1,1,1]
   }
   n26 = "Constraint - dyn_strip_r_D_MSTR (PPPPPPP)"!(
    ComponentPose=@n25
    Alpha=1
    ConstraintWeights_0=0.0726
    ConstraintWeights_1:float=0.0372
    ConstraintWeights_2:float=0.1877
    ConstraintWeights_3:float=0.0364
    ConstraintWeights_4:float=0.0485
    ConstraintWeights_5:float=0.4151
    ConstraintWeights_6:float=0.2025
   ) -> (Pose) {
    alphaBoolBlend {blendOption=Linear}
    boneToModify {boneName=dyn_strip_r_D_MSTR}
    constraintSetup [
     {targetBone {boneName=deform_thigh_bk_r}}
     {targetBone {boneName=dyn_loincloth_bk_r_02}}
     {targetBone {boneName=thigh_twist_01_r}}
     {targetBone {boneName=dyn_loincloth_fr_r_MSTR}}
     {targetBone {boneName=spine_01}}
     {targetBone {boneName=pelvis}}
     {targetBone {boneName=deform_thigh_fr_r}}
    ]
    constraintWeights [1,1,1,1,1,1,1]
   }
   n27 = "Constraint - dyn_strip_r_E_MSTR (PPPPPPP)"!(
    ComponentPose=@n26
    Alpha=1
    ConstraintWeights_0=0.1079
    ConstraintWeights_1:float=0.1726
    ConstraintWeights_2:float=0.0114
    ConstraintWeights_3:float=0.1094
    ConstraintWeights_4:float=0.0496
    ConstraintWeights_5:float=0.4414
    ConstraintWeights_6:float=0.1077
   ) -> (Pose) {
    alphaBoolBlend {blendOption=Linear}
    boneToModify {boneName=dyn_strip_r_E_MSTR}
    constraintSetup [
     {targetBone {boneName=deform_thigh_bk_r}}
     {targetBone {boneName=dyn_loincloth_bk_r_02}}
     {targetBone {boneName=dyn_loincloth_bk_r_01}}
     {targetBone {boneName=thigh_twist_01_r}}
     {targetBone {boneName=spine_01}}
     {targetBone {boneName=pelvis}}
     {targetBone {boneName=deform_thigh_fr_r}}
    ]
    constraintWeights [1,1,1,1,1,1,1]
   }
   n28 = "Constraint - dyn_strip_r_F_MSTR (PPPPPPPPPP)"!(
    ComponentPose=@n27
    Alpha=1
    ConstraintWeights_0=0.0491
    ConstraintWeights_1:float=0.5673
    ConstraintWeights_2:float=0.0297
    ConstraintWeights_3:float=0.038
    ConstraintWeights_4:float=0.014
    ConstraintWeights_5:float=0.024
    ConstraintWeights_6:float=0.022
    ConstraintWeights_7:float=0.0137
    ConstraintWeights_8:float=0.208
    ConstraintWeights_9:float=0.0343
   ) -> (Pose) {
    alphaBoolBlend {blendOption=Linear}
    boneToModify {boneName=dyn_strip_r_F_MSTR}
    constraintSetup [
     {targetBone {boneName=deform_thigh_bk_r}}
     {targetBone {boneName=dyn_loincloth_bk_r_02}}
     {targetBone {boneName=dyn_loincloth_bk_r_01}}
     {targetBone {boneName=thigh_twist_01_r}}
     {targetBone {boneName=dyn_loincloth_bk_04}}
     {targetBone {boneName=spine_01}}
     {targetBone {boneName=dyn_loincloth_bk_02}}
     {targetBone {boneName=dyn_loincloth_bk_03}}
     {targetBone {boneName=pelvis}}
     {targetBone {boneName=deform_thigh_fr_r}}
    ]
    constraintWeights [1,1,1,1,1,1,1,1,1,1]
   }
   n29 = "Constraint - dyn_strip_r_G_MSTR (PPPPPP)"!(
    ComponentPose=@n28
    Alpha=1
    ConstraintWeights_0=0.8493
    ConstraintWeights_1:float=0.0182
    ConstraintWeights_2:float=0.0464
    ConstraintWeights_3:float=0.0217
    ConstraintWeights_4:float=0.0208
    ConstraintWeights_5:float=0.0435
   ) -> (Pose) {
    alphaBoolBlend {blendOption=Linear}
    boneToModify {boneName=dyn_strip_r_G_MSTR}
    constraintSetup [
     {targetBone {boneName=dyn_loincloth_bk_r_02}}
     {targetBone {boneName=dyn_loincloth_bk_r_01}}
     {targetBone {boneName=dyn_loincloth_bk_04}}
     {targetBone {boneName=dyn_loincloth_bk_02}}
     {targetBone {boneName=dyn_loincloth_bk_03}}
     {targetBone {boneName=pelvis}}
    ]
    constraintWeights [1,1,1,1,1,1]
   }
   n30 = "Transform (Modify) Bone - Bone: dyn_back_fur_01_MSTR"(ComponentPose=@n29, Alpha=1, Translation="-0.5,0,2", Rotation="0,0,0", Scale="1,1,1") {
    rotationMode=BMM_Additive
    rotationSpace=BCS_BoneSpace
    translationMode=BMM_Additive
    translationSpace=BCS_BoneSpace
    alphaBoolBlend {blendOption=Linear}
    boneToModify {boneName=dyn_back_fur_01_MSTR}
    translation {x=-0.5, z=2}
   }
   n31 = Anim Dynamics - Bone: dyn_omega_01_MSTR(ComponentPose=@n30, Alpha=1, ExternalForce="0,0,0") {
    angularDampingOverride=0.995
    bGravityOverrideInSimSpace=true
    bOverrideAngularDamping=true
    bUseGravityOverride=true
    simulationSpace=RootRelative
    alphaBoolBlend {blendOption=Linear}
    alphaScaleBiasClamp {bInterpResult=true}
    boundBone {boneName=dyn_omega_01_MSTR}
    componentAppliedLinearAccClamp=<3000 3000 3000>
    componentLinearAccScale=<0.25 0.25 0.3>
    componentLinearVelScale=<0.125 0.125 0.15>
    gravityOverride {z=-850}
    physicsBodyDefinitions [{
      sphereCollisionRadius=0
      boundBone {boneName=dyn_omega_01_MSTR}
      boxExtents=<15 15 2>
      constraintSetup {angularLimitsMax=<5 1 15>, angularLimitsMin=<-10 -75 -20>, angularTarget=<0 0 0>}
      localJointOffset=<3 0 0>
     }]
    relativeSpaceBone {boneName=spine_05}
   }
   n32 = "Anim Dynamics - Chain: dyn_back_fur_01 -> dyn_back_fur_02"(ComponentPose=@n31, Alpha=Get Emoting Alpha() -> (Emoting Alpha:double), ExternalForce="0,0,0") {
    angularDampingOverride=0.999
    bChain=true
    bGravityOverrideInSimSpace=true
    bOverrideAngularDamping=true
    bUseGravityOverride=true
    alphaBoolBlend {blendOption=Linear}
    alphaScaleBiasClamp {bInterpResult=true}
    boundBone {boneName=dyn_back_fur_01}
    chainEnd {boneName=dyn_back_fur_02}
    componentAppliedLinearAccClamp=<3000 3000 3000>
    componentLinearAccScale=<0.2 0.2 0.35>
    componentLinearVelScale=<0.2 0.2 0.2>
    gravityOverride {z=-1000}
    physicsBodyDefinitions [{
      sphereCollisionRadius=0
      boundBone {boneName=dyn_back_fur_01}
      boxExtents=<20 20 8>
      constraintSetup {angularLimitsMax=<5 0 5>, angularLimitsMin=<-5 -45 -5>, angularTarget=<0 0 0>}
      localJointOffset=<3 0 0>
     }]
    relativeSpaceBone {boneName=dyn_back_fur_01_MSTR}
   }
   n33 = Anim Dynamics - Bone: dyn_side_fur_01(ComponentPose=@n32, Alpha=1, ExternalForce="0,0,0") {
    angularDampingOverride=0.999
    bGravityOverrideInSimSpace=true
    bOverrideAngularDamping=true
    bUseGravityOverride=true
    alphaBoolBlend {blendOption=Linear}
    alphaScaleBiasClamp {bInterpResult=true}
    boundBone {boneName=dyn_side_fur_01}
    chainEnd {boneName=dyn_back_fur_02}
    componentAppliedLinearAccClamp=<3000 3000 3000>
    componentLinearAccScale=<0.2 0.2 0.35>
    componentLinearVelScale=<0.2 0.2 0.2>
    gravityOverride {z=-800}
    physicsBodyDefinitions [{
      sphereCollisionRadius=0
      boundBone {boneName=dyn_side_fur_01}
      boxExtents=<20 20 8>
      constraintSetup {angularLimitsMax=<5 0 10>, angularLimitsMin=<-5 -45 -10>, angularTarget=<0 0 0>}
      localJointOffset=<3 0 0>
     }]
    relativeSpaceBone {boneName=dyn_side_fur_01_MSTR}
   }
   n34 = Anim Dynamics - Bone: dyn_horn_r_01(ComponentPose=@n33, Alpha=1, ExternalForce="0,0,0") {
    angularDampingOverride=0.994
    bGravityOverrideInSimSpace=true
    bOverrideAngularDamping=true
    bUseGravityOverride=true
    simulationSpace=RootRelative
    alphaBoolBlend {blendOption=Linear}
    alphaScaleBiasClamp {bInterpResult=true}
    boundBone {boneName=dyn_horn_r_01}
    componentAppliedLinearAccClamp=<3000 3000 3000>
    componentLinearAccScale=<0.2 0.2 0.4>
    componentLinearVelScale=<0.1 0.1 0.125>
    gravityOverride {z=-1000}
    physicsBodyDefinitions [{
      sphereCollisionRadius=0
      boundBone {boneName=dyn_horn_r_01}
      boxExtents=<20 20 10>
      constraintSetup {angularLimitsMax=<0 -5 20>, angularLimitsMin=<-5 -70 -10>, angularTarget=<0 0 0>}
      localJointOffset=<3 0 0>
     }]
    relativeSpaceBone {boneName=pelvis}
   }
   n35 = Anim Dynamics - Bone: dyn_knife_l_01(ComponentPose=@n34, Alpha=1, ExternalForce="0,0,0") {
    angularDampingOverride=0.999
    bGravityOverrideInSimSpace=true
    bOverrideAngularDamping=true
    bUseGravityOverride=true
    simulationSpace=RootRelative
    alphaBoolBlend {blendOption=Linear}
    alphaScaleBiasClamp {bInterpResult=true}
    boundBone {boneName=dyn_knife_l_01}
    chainEnd {boneName=dyn_knife_l_02}
    componentAppliedLinearAccClamp=<3000 3000 3000>
    componentLinearAccScale=<0.2 0.2 0.3>
    componentLinearVelScale=<0.1 0.1 0.125>
    gravityOverride {z=-980}
    physicsBodyDefinitions [{
      sphereCollisionRadius=0
      boundBone {boneName=dyn_knife_l_01}
      boxExtents=<25 15 10>
      constraintSetup {angularLimitsMax=<5 0 5>, angularLimitsMin=<-2 -15 -5>, angularTarget=<0 0 0>}
      localJointOffset=<3 0 0>
     }]
    relativeSpaceBone {boneName=pelvis}
   }
   n36 = Anim Dynamics - Bone: dyn_knife_l_02(ComponentPose=@n35, Alpha=1, ExternalForce="0,0,0") {
    angularDampingOverride=0.999
    bGravityOverrideInSimSpace=true
    bOverrideAngularDamping=true
    bUseGravityOverride=true
    simulationSpace=RootRelative
    alphaBoolBlend {blendOption=Linear}
    alphaScaleBiasClamp {bInterpResult=true}
    boundBone {boneName=dyn_knife_l_02}
    chainEnd {boneName=dyn_knife_l_02}
    componentAppliedLinearAccClamp=<3000 3000 3000>
    componentLinearAccScale=<0.4 0.4 0.4>
    componentLinearVelScale=<0.2 0.2 0.2>
    gravityOverride {z=-980}
    physicsBodyDefinitions [{
      sphereCollisionRadius=0
      boundBone {boneName=dyn_knife_l_02}
      boxExtents=<25 15 10>
      constraintSetup {angularLimitsMax=<5 0 5>, angularLimitsMin=<-2 -15 -5>, angularTarget=<0 0 0>}
      localJointOffset=<3 0 0>
     }]
    relativeSpaceBone {boneName=dyn_knife_l_01}
   }
   n37 = Anim Dynamics - Bone: dyn_satchel_l_01(ComponentPose=@n36, Alpha=1, ExternalForce="0,0,0") {
    angularDampingOverride=0.999
    bGravityOverrideInSimSpace=true
    bOverrideAngularDamping=true
    bUseGravityOverride=true
    simulationSpace=RootRelative
    alphaBoolBlend {blendOption=Linear}
    alphaScaleBiasClamp {bInterpResult=true}
    boundBone {boneName=dyn_satchel_l_01}
    chainEnd {boneName=dyn_satchel_l_02}
    componentAppliedLinearAccClamp=<3000 3000 3000>
    componentLinearAccScale=<0.2 0.2 0.35>
    componentLinearVelScale=<0.1 0.1 0.125>
    gravityOverride {z=-980}
    physicsBodyDefinitions [{
      sphereCollisionRadius=0
      boundBone {boneName=dyn_satchel_l_01}
      boxExtents=<15 25 10>
      constraintSetup {angularLimitsMax=<5 0 2>, angularLimitsMin=<-5 -35 -2>, angularTarget=<0 0 0>}
      localJointOffset=<3 0 0>
     }]
    relativeSpaceBone {boneName=pelvis}
   }
   n38 = Anim Dynamics - Bone: dyn_satchel_l_02(ComponentPose=@n37, Alpha=1, ExternalForce="0,0,0") {
    angularDampingOverride=0.999
    bGravityOverrideInSimSpace=true
    bOverrideAngularDamping=true
    bUseGravityOverride=true
    simulationSpace=RootRelative
    alphaBoolBlend {blendOption=Linear}
    alphaScaleBiasClamp {bInterpResult=true}
    boundBone {boneName=dyn_satchel_l_02}
    chainEnd {boneName=dyn_satchel_l_02}
    componentAppliedLinearAccClamp=<3000 3000 3000>
    componentLinearAccScale=<0.4 0.4 0.4>
    componentLinearVelScale=<0.2 0.2 0.2>
    gravityOverride {z=-980}
    physicsBodyDefinitions [{
      sphereCollisionRadius=0
      boundBone {boneName=dyn_satchel_l_02}
      boxExtents=<15 25 10>
      constraintSetup {angularLimitsMax=<5 0 2>, angularLimitsMin=<-5 -35 -2>, angularTarget=<0 0 0>}
      localJointOffset=<3 0 0>
     }]
    relativeSpaceBone {boneName=dyn_satchel_l_01}
   }
   n39 = Anim Dynamics - Bone: dyn_bag_l_01(ComponentPose=@n38, Alpha=1, ExternalForce="0,0,0") {
    angularDampingOverride=0.999
    bGravityOverrideInSimSpace=true
    bOverrideAngularDamping=true
    bUseGravityOverride=true
    simulationSpace=RootRelative
    alphaBoolBlend {blendOption=Linear}
    alphaScaleBiasClamp {bInterpResult=true}
    boundBone {boneName=dyn_bag_l_01}
    chainEnd {boneName=dyn_bag_l_03}
    componentAppliedLinearAccClamp=<3000 3000 3000>
    componentLinearAccScale=<0.2 0.2 0.3>
    componentLinearVelScale=<0.1 0.1 0.125>
    gravityOverride {z=-980}
    physicsBodyDefinitions [{
      sphereCollisionRadius=0
      boundBone {boneName=dyn_bag_l_01}
      boxExtents=<25 25 10>
      constraintSetup {angularLimitsMax=<5 0 5>, angularLimitsMin=<-5 -25 -5>, angularTarget=<0 0 0>}
      localJointOffset=<3 0 0>
     }]
    relativeSpaceBone {boneName=pelvis}
   }
   n40 = Anim Dynamics - Bone: dyn_bag_l_02(ComponentPose=@n39, Alpha=1, ExternalForce="0,0,0") {
    angularDampingOverride=0.999
    bGravityOverrideInSimSpace=true
    bOverrideAngularDamping=true
    bUseGravityOverride=true
    simulationSpace=RootRelative
    alphaBoolBlend {blendOption=Linear}
    alphaScaleBiasClamp {bInterpResult=true}
    boundBone {boneName=dyn_bag_l_02}
    chainEnd {boneName=dyn_bag_l_03}
    componentAppliedLinearAccClamp=<3000 3000 3000>
    componentLinearAccScale=<0.4 0.4 0.4>
    componentLinearVelScale=<0.2 0.2 0.2>
    gravityOverride {z=-980}
    physicsBodyDefinitions [{
      sphereCollisionRadius=0
      boundBone {boneName=dyn_bag_l_02}
      boxExtents=<25 25 10>
      constraintSetup {angularLimitsMax=<5 0 5>, angularLimitsMin=<-5 -25 -5>, angularTarget=<0 0 0>}
      localJointOffset=<3 0 0>
     }]
    relativeSpaceBone {boneName=dyn_bag_l_01}
   }
   n41 = Anim Dynamics - Bone: dyn_bag_l_03(ComponentPose=@n40, Alpha=1, ExternalForce="0,0,0") {
    angularDampingOverride=0.999
    bGravityOverrideInSimSpace=true
    bOverrideAngularDamping=true
    bUseGravityOverride=true
    simulationSpace=RootRelative
    alphaBoolBlend {blendOption=Linear}
    alphaScaleBiasClamp {bInterpResult=true}
    boundBone {boneName=dyn_bag_l_03}
    chainEnd {boneName=dyn_bag_l_03}
    componentAppliedLinearAccClamp=<3000 3000 3000>
    componentLinearAccScale=<0.4 0.4 0.4>
    componentLinearVelScale=<0.2 0.2 0.2>
    gravityOverride {z=-980}
    physicsBodyDefinitions [{
      sphereCollisionRadius=0
      boundBone {boneName=dyn_bag_l_03}
      boxExtents=<25 25 10>
      constraintSetup {angularLimitsMax=<5 0 5>, angularLimitsMin=<-5 -25 -5>, angularTarget=<0 0 0>}
      localJointOffset=<3 0 0>
     }]
    relativeSpaceBone {boneName=dyn_bag_l_02}
   }
   n42 = "Transform (Modify) Bone - Bone: dyn_bag_l_03"!(ComponentPose=@n41, bAlphaBoolEnabled:bool=True, Translation="0,0,1", Rotation="15,0,0", Scale="1,1,1") -> (Pose) {
    alphaInputType=Bool
    bAlphaBoolEnabled=false
    rotationMode=BMM_Additive
    rotationSpace=BCS_BoneSpace
    translationMode=BMM_Additive
    translationSpace=BCS_BoneSpace
    alphaBoolBlend {blendOption=Linear}
    boneToModify {boneName=dyn_bag_l_03}
    rotation {pitch=15}
    translation {z=1}
   }
   n43 = Get Skydive Belt Alpha() -> (Skydive Belt Alpha:double)
   n44 = "Transform (Modify) Bone - Bone: dyn_knife_l_01"(ComponentPose=@n42, Alpha=@n43, Translation="-2,2,4", Rotation="0.000,0.000,30", Scale="1,1,1") {
    rotationMode=BMM_Additive
    rotationSpace=BCS_BoneSpace
    translationMode=BMM_Additive
    translationSpace=BCS_BoneSpace
    alphaBoolBlend {blendOption=Linear}
    boneToModify {boneName=dyn_knife_l_01}
    rotation {roll=30}
    translation=<-2 2 4>
   }
   n45 = "Constraint - dyn_horn_r_MSTR (P)"(ComponentPose=@n44, Alpha=@n43, ConstraintWeights_0=1) {
    alphaBoolBlend {blendOption=Linear}
    boneToModify {boneName=dyn_horn_r_MSTR}
    constraintSetup [{targetBone {boneName=dyn_loincloth_bk_r_01}}]
    constraintWeights [1]
   }
   n46 = "Constraint - dyn_satchel_l_MSTR (P)"(ComponentPose=@n45, Alpha=@n43, ConstraintWeights_0=1) {
    alphaBoolBlend {blendOption=Linear}
    boneToModify {boneName=dyn_satchel_l_MSTR}
    constraintSetup [{targetBone {boneName=dyn_loincloth_bk_01}}]
    constraintWeights [1]
   }
   n47 = "Constraint - dyn_belt_strap_fr_01 (P)"(ComponentPose=@n46, Alpha=0.8, ConstraintWeights_0=1) {
    alphaBoolBlend {blendOption=Linear}
    boneToModify {boneName=dyn_belt_strap_fr_01}
    constraintSetup [{targetBone {boneName=dyn_loincloth_fr_01}}]
    constraintWeights [1]
   }
   n48 = "Constraint - dyn_belt_strap_fr_02 (P)"(ComponentPose=@n47, Alpha=0.8, ConstraintWeights_0=1) {
    alphaBoolBlend {blendOption=Linear}
    boneToModify {boneName=dyn_belt_strap_fr_02}
    constraintSetup [{targetBone {boneName=dyn_loincloth_fr_02}}]
    constraintWeights [1]
   }
   n49 = "Constraint - dyn_belt_strap_fr_03 (P)"(ComponentPose=@n48, Alpha=0.8, ConstraintWeights_0=1) {
    alphaBoolBlend {blendOption=Linear}
    boneToModify {boneName=dyn_belt_strap_fr_03}
    constraintSetup [{targetBone {boneName=dyn_loincloth_fr_03}}]
    constraintWeights [1]
   }
   n50 = "Transform (Modify) Bone - Bone: dyn_belt_strap_fr_01"(ComponentPose=@n49, Alpha=1, Translation="0,0,1", Rotation="5,0,0", Scale="1,1,1") {
    rotationMode=BMM_Additive
    rotationSpace=BCS_BoneSpace
    translationMode=BMM_Additive
    translationSpace=BCS_BoneSpace
    alphaBoolBlend {blendOption=Linear}
    boneToModify {boneName=dyn_belt_strap_fr_01}
    rotation {pitch=5}
    translation {z=1}
   }
   n51 = "Anim Dynamics - Chain: dyn_belt_strap_fr_02 -> dyn_belt_strap_fr_03"(ComponentPose=@n50, Alpha=0.4, ExternalForce="0,0,0") {
    angularDampingOverride=0.999
    bChain=true
    bGravityOverrideInSimSpace=true
    bOverrideAngularDamping=true
    bUseGravityOverride=true
    alphaBoolBlend {blendOption=Linear}
    boundBone {boneName=dyn_belt_strap_fr_02}
    chainEnd {boneName=dyn_belt_strap_fr_03}
    componentAppliedLinearAccClamp=<3000 3000 3000>
    componentLinearAccScale=<0.2 0.2 0.35>
    componentLinearVelScale=<0.1 0.1 0.125>
    gravityOverride {z=-980}
    physicsBodyDefinitions [{
      sphereCollisionRadius=0
      boundBone {boneName=dyn_belt_strap_fr_02}
      boxExtents=<15 25 10>
      constraintSetup {angularLimitsMax=<5 0 10>, angularLimitsMin=<-5 -35 -10>, angularTarget=<0 0 0>}
      localJointOffset=<3 0 0>
     }]
   }
   n52 = AND Boolean(A=Get bIsInFrontEnd(), B=NOT Boolean(A=Get bIsPlayingEmote()))
   n53 = SelectFloat(A=0, B=0.8, bPickA=@n52)
   n54 = "Constraint - dyn_strip_r_A_MSTR (P)"(ComponentPose=@n51, Alpha=@n53, ConstraintWeights_0=1) {
    alphaBoolBlend {blendOption=Linear}
    alphaScaleBiasClamp {bInterpResult=true}
    boneToModify {boneName=dyn_strip_r_A_MSTR}
    constraintSetup [{targetBone {boneName=thigh_r}}]
    constraintWeights [1]
   }
   n55 = "Constraint - dyn_strip_r_B_MSTR (P)"(ComponentPose=@n54, Alpha=@n53, ConstraintWeights_0=1) {
    alphaBoolBlend {blendOption=Linear}
    alphaScaleBiasClamp {bInterpResult=true}
    boneToModify {boneName=dyn_strip_r_B_MSTR}
    constraintSetup [{targetBone {boneName=thigh_r}}]
    constraintWeights [1]
   }
   n56 = SelectFloat(A=0, B=0.5, bPickA=@n52)
   n57 = "Constraint - dyn_strip_r_C_MSTR (P)"(ComponentPose=@n55, Alpha=@n56, ConstraintWeights_0=1) {
    alphaBoolBlend {blendOption=Linear}
    alphaScaleBiasClamp {bInterpResult=true}
    boneToModify {boneName=dyn_strip_r_C_MSTR}
    constraintSetup [{targetBone {boneName=thigh_r}}]
    constraintWeights [1]
   }
   n58 = "Constraint - dyn_strip_r_D_MSTR (P)"(ComponentPose=@n57, Alpha=0.5, ConstraintWeights_0=1) {
    alphaBoolBlend {blendOption=Linear}
    alphaScaleBiasClamp {bInterpResult=true}
    boneToModify {boneName=dyn_strip_r_D_MSTR}
    constraintSetup [{targetBone {boneName=thigh_r}}]
    constraintWeights [1]
   }
   n59 = "Constraint - dyn_strip_r_E_MSTR (P)"(ComponentPose=@n58, Alpha=0.5, ConstraintWeights_0=1) {
    alphaBoolBlend {blendOption=Linear}
    alphaScaleBiasClamp {bInterpResult=true}
    boneToModify {boneName=dyn_strip_r_E_MSTR}
    constraintSetup [{targetBone {boneName=thigh_r}}]
    constraintWeights [1]
   }
   n60 = "Constraint - dyn_strip_r_F_MSTR (P)"(ComponentPose=@n59, Alpha=0.5, ConstraintWeights_0=1) {
    alphaBoolBlend {blendOption=Linear}
    alphaScaleBiasClamp {bInterpResult=true}
    boneToModify {boneName=dyn_strip_r_F_MSTR}
    constraintSetup [{targetBone {boneName=thigh_r}}]
    constraintWeights [1]
   }
   n61 = "Constraint - dyn_strip_l_A_MSTR (P)"(ComponentPose=@n60, Alpha=@n56, ConstraintWeights_0=1) {
    alphaBoolBlend {blendOption=Linear}
    alphaScaleBiasClamp {bInterpResult=true}
    boneToModify {boneName=dyn_strip_l_A_MSTR}
    constraintSetup [{targetBone {boneName=thigh_l}}]
    constraintWeights [1]
   }
   n62 = "Constraint - dyn_strip_l_B_MSTR (P)"(ComponentPose=@n61, Alpha=@n56, ConstraintWeights_0=1) {
    alphaBoolBlend {blendOption=Linear}
    alphaScaleBiasClamp {bInterpResult=true}
    boneToModify {boneName=dyn_strip_l_B_MSTR}
    constraintSetup [{targetBone {boneName=thigh_l}}]
    constraintWeights [1]
   }
   n63 = "Constraint - dyn_strip_l_C_MSTR (P)"(ComponentPose=@n62, Alpha=@n56, ConstraintWeights_0=1) {
    alphaBoolBlend {blendOption=Linear}
    alphaScaleBiasClamp {bInterpResult=true}
    boneToModify {boneName=dyn_strip_l_C_MSTR}
    constraintSetup [{targetBone {boneName=dyn_strip_l_B_MSTR}}]
    constraintWeights [1]
   }
   n64 = "Constraint - dyn_strip_l_D_MSTR (P)"(ComponentPose=@n63, Alpha=@n56, ConstraintWeights_0=1) {
    alphaBoolBlend {blendOption=Linear}
    alphaScaleBiasClamp {bInterpResult=true}
    boneToModify {boneName=dyn_strip_l_D_MSTR}
    constraintSetup [{targetBone {boneName=thigh_l}}]
    constraintWeights [1]
   }
   n65 = "Constraint - dyn_strip_l_E_MSTR (P)"(ComponentPose=@n64, Alpha=@n56, ConstraintWeights_0=1) {
    alphaBoolBlend {blendOption=Linear}
    alphaScaleBiasClamp {bInterpResult=true}
    boneToModify {boneName=dyn_strip_l_E_MSTR}
    constraintSetup [{targetBone {boneName=thigh_l}}]
    constraintWeights [1]
   }
   n66 = "Constraint - dyn_strip_l_F_MSTR (P)"(ComponentPose=@n65, Alpha=@n56, ConstraintWeights_0=1) {
    alphaBoolBlend {blendOption=Linear}
    alphaScaleBiasClamp {bInterpResult=true}
    boneToModify {boneName=dyn_strip_l_F_MSTR}
    constraintSetup [{targetBone {boneName=thigh_l}}]
    constraintWeights [1]
   }
   n67 = Get bIsPlayingEmote()
   n68 = "Transform (Modify) Bone - Bone: dyn_strip_l_C_MSTR"!(
    ComponentPose=@n66
    bAlphaBoolEnabled:bool=AND Boolean(A=Get bIsInFrontEnd(), B=NOT Boolean(A=@n67))
    Translation="0,-0.75,0"
    Rotation="0,0,0"
    Scale="1,1,1"
   ) -> (Pose) {
    alphaInputType=Bool
    rotationMode=BMM_Additive
    rotationSpace=BCS_BoneSpace
    translationMode=BMM_Additive
    translationSpace=BCS_BoneSpace
    alphaBoolBlend {blendOption=Linear}
    boneToModify {boneName=dyn_strip_l_C_MSTR}
    translation {y=-0.75}
   }
   n69 = Control Rig(Source=Component To Local(ComponentPose=@n68)) {
    alphaBoolBlend {blendOption=Linear}
    controlRigAssetReference {blueprintRigClass=&/M_MED_Historian_Deform_CtrlRig_side_strips.M_MED_Historian_Deform_CtrlRig_side_strips_C}
   }
   n70 = RigidBody(ComponentPose=Local To Component(LocalPose=@n69), OverrideWorldGravity=@n1) {
    bClampLinearTranslationLimitToRefPose=true
    bOverrideWorldGravity=true
    overridePhysicsAsset=&/M_MED_Historian_Dyn_strips
    simulationSpace=BaseBoneSpace
    alphaBoolBlend {blendOption=Linear}
    baseBoneRef {boneName=dyn_root_simspace}
    componentAppliedLinearAccClamp=<3000 3000 3000>
    componentLinearAccScale {z=0.5}
    componentLinearVelScale {z=0.5}
    overrideWorldGravity {z=-980}
    simSpaceSettings {
     maxAngularAcceleration=1000
     maxAngularVelocity=3
     maxLinearAcceleration=250
     maxLinearVelocity=500
     velocityScaleZ=0
     worldAlpha=1
     externalLinearDragV=<0.1 0.1>
    }
   }
   Output Pose(Result=Component To Local(ComponentPose=@n70)) {layerGroup=None}
  ]}
 Complex Sim Space {nodes [
   n1 = vector + vector(
    A:Vector=GetSocketLocation!(self:Scene Component Object Reference=Get Skel_Mesh(), InSocketName:Name=root) -> (ReturnValue:Vector)
    B:Vector="0, 0,165"
   ) -> (ReturnValue:Vector) {is=title K2Node_CommutativeAssociativeBinaryOperator}
   n2 = Complex Sim Space!() -> (then, InterpSpeedMax:double=~, PawnSpeedMax:double=~, RInterpErrorTolerance:double=~, DeltaTime:double=~) {is=title K2Node_FunctionEntry}
   n3 = Get target_sim_space() -> (target_sim_space:Rotator)
   n4 = Get root_space_rotation()
   n5 = Set root_space_rotation(
    execute=[@n20.then, @n21.then]
    root_space_rotation:Rotator=RInterpTo!(Current:Rotator=@n4, Target:Rotator=@n3, DeltaTime:float=@n2.DeltaTime, InterpSpeed:float=Get InterpSpeed() -> (InterpSpeed:double)) -> (ReturnValue:Rotator)
   ) -> (Output_Get:Rotator)
   n6 = DrawDebugCoordinateSystem!(
    execute
    AxisLoc:Vector=@n1
    AxisRot:Rotator=@n5
    Scale:float=35
    Duration:float=0
    Thickness:float=3
    DepthPriority:EDrawDebugSceneDepthPriorityGroup Enum=World
   ) -> (then)
   Set inComponentSpaceLastFrame(execute=[@n6.then, @n5.then], inComponentSpaceLastFrame:bool=Get inComponentSpace() -> (inComponentSpace:bool)) -> (Output_Get:bool)
   > Branch(Condition="Equal (Rotator)"!(A:Rotator=@n4, B:Rotator=@n3, ErrorTolerance:float=@n2.RInterpErrorTolerance) -> (ReturnValue:bool))
   > Set lockedOnSimSpace(lockedOnSimSpace:bool=true) -> (Output_Get:bool)
   > n7 = Set InterpSpeed(InterpSpeed:double=0) -> (Output_Get:double)
   DrawDebugCoordinateSystem(AxisLoc=@n1, AxisRot=Get root_rotation() -> (root_rotation:Rotator), Scale=75, Duration=0, Thickness=1.5, DepthPriority=World)
   PrintString!(execute, InString="To String (Float)"(InDouble=@n7), bPrintToScreen, bPrintToLog, TextColor, Duration, Key) -> (then)
   n8 = GetOwningActor()
   n9 = Get Actor Location(self=@n8)
   n10 = vector * float(A=GetActorForwardVector!(self:Actor Object Reference=@n8) -> (ReturnValue:Vector), B=500)
   DrawDebugLine!(
    execute
    LineStart:Vector=vector + vector(A=@n9, B="0,0,87")
    LineEnd:Vector=vector + vector(A=vector + vector(A=@n9, B=@n10), B="0,0,87")
    LineColor:Linear Color Structure=(R=0.6348,G=0,B=0.6406,A=1)
    Duration:float=0
    Thickness:float=0.5
    DepthPriority:EDrawDebugSceneDepthPriorityGroup Enum=World
   ) -> (then)
   Reroute Node(InputPin=@n16.then_2)
   Reroute Node(InputPin=@n16.then_1)
   Set root_rotation(
    execute=@n16.then_0
    root_rotation:Rotator=GetSocketRotation!(self:Scene Component Object Reference=Get Skel_Mesh(), InSocketName:Name=root) -> (ReturnValue:Rotator)
   ) -> (Output_Get:Rotator)
   n11 = GetOwningActor()
   n12 = "Equal (Boolean)"!(A:bool=Get inComponentSpaceLastFrame() -> (inComponentSpaceLastFrame:bool), B:bool=Get inComponentSpace()) -> (ReturnValue:bool)
   n13 = Get Actor Rotation!(self:Actor Object Reference=@n11) -> (ReturnValue:Rotator) {is=title K2Node_CallFunction}
   n14 = "float > float"(A=Get PawnSpeed2D(), B=@n2.PawnSpeedMax) {is=title K2Node_CallFunction}
   n15 = Combine Rotators!(A:Rotator=@n13, B:Rotator="0,-90, 0") -> (ReturnValue:Rotator)
   n16 = Sequence!(execute=@n2.then) -> (then_0, then_1, then_2:Exec, then_3:Exec)
   Set InterpSpeedMax(execute=@n16.then_3, InterpSpeedMax:double=@n2) -> (Output_Get:double)
   > n17 = Set inComponentSpace(inComponentSpace:bool=@n14) -> (Output_Get:bool)
   > n18 = Branch(Condition=@n17)
   > n19 = Set target_sim_space(target_sim_space:Rotator=@n15) -> (Output_Get:Rotator)
   n20 = Branch(execute=[@n22.then, @n19.then], Condition=@n12)
   Set lockedOnSimSpace(execute=@n20.else, lockedOnSimSpace=false)
   > n21 = Set InterpSpeed(InterpSpeed=Get InterpSpeedMax() -> (InterpSpeedMax:double))
   n22 = Set target_sim_space(execute=@n18.else, target_sim_space=Get root_rotation())
  ]}
 EventGraph {nodes [
   n1 = Event BlueprintUpdateAnimation()
   > n2 = Sequence()
   > n3 = Branch(Condition=Get bIsSwimSprinting() -> (bIsSwimSprinting:bool))
   n4 = GetSocketRotation(self=Get Skel_Mesh(), InSocketName=pelvis)
   n5 = PerlinNoise1D!(Value:float=GetGameTimeInSeconds()) -> (ReturnValue:float)
   n6 = float * float(A="Sin (Degrees)"!(A:double=float * float(A=@n5, B=800)) -> (ReturnValue:double), B=-400)
   n7 = vector + vector(
    A=vector * float(A=GetForwardVector!(InRot:Rotator=@n4) -> (ReturnValue:Vector), B=-300)
    B=MakeVector(Y=0, Z=float + float(A=@n6, B=200) {is=title K2Node_CommutativeAssociativeBinaryOperator})
   )
   > Set GravityOverride(GravityOverride=@n7)
   n8 = Branch(execute=@n3.else, Condition=Get bIsSkyDiving())
   > n9 = Branch(Condition=Get bIsParachuteOpened())
   Set GravityOverride(execute=[@n8.else, @n9.then], GravityOverride="0, 0,-980")
   n10 = Branch(execute=@n9.else, Condition="float > float"(A=Get PawnVelocityZ(), B=50))
   > Set GravityOverride(GravityOverride="0, 0,-100")
   Set GravityOverride(execute=@n10.else, GravityOverride=@n7)
   Complex Sim Space!(execute=@n2.then_1, self, InterpSpeedMax:double=9, PawnSpeedMax:double=20, RInterpErrorTolerance:double=1, DeltaTime:double=@n1.DeltaTimeX) -> (then) {is=title K2Node_CallFunction}
   Set Skydive Belt Alpha(execute=@n2.then_2, Skydive Belt Alpha:double=SelectFloat(A=0.5, B=0, bPickA=Get bIsSkyDiving())) -> (Output_Get:double)
   > Set Skydive Loincloth Alpha(Skydive Loincloth Alpha:double=SelectFloat(A=0.5, B=1, bPickA=Get bIsSkydiveFloating())) -> (Output_Get:double)
   > Set Emoting Alpha(Emoting Alpha:double=SelectFloat(A=0.5, B=1, bPickA=Get bIsPlayingEmote())) -> (Output_Get:double)
   Event BlueprintInitializeAnimation()
   > Set Skel_Mesh(Skel_Mesh=GetOwnerSkeletalMeshComponent())
   > Set InterpSpeed(InterpSpeed=Get InterpSpeedMax())
   Event OnEndCrouching()
   > Set crouch alpha(crouch alpha:double=0) -> (Output_Get:double)
   Event OnBeginCrouching()
   > Set crouch alpha(crouch alpha=1)
   n11 = Get Skel_Mesh()
   DrawDebugCoordinateSystem(
    AxisLoc=vector + vector(A=GetSocketLocation(self=@n11, InSocketName=dyn_root_simspace), B="0, 0,170")
    AxisRot=Get root_rotation()
    Scale=75
    Duration=0
    Thickness=1.5
    DepthPriority=World
   )
  ]}
]
variables [
 crouch alpha {type=number}
 Emoting Alpha {type=number}
 GravityOverride {type=Vector}
 Idle Standing Alpha {type=number}
 inComponentSpace {type=boolean}
 inComponentSpaceLastFrame {type=boolean}
 InterpSpeed {type=number}
 InterpSpeedMax {type=number}
 lockedOnSimSpace {type=boolean}
 root_rotation {type=Rotator}
 root_space_rotation {type=Rotator}
 Skel_Mesh {type=SkeletalMeshComponent}
 Skydive Belt Alpha {type=number}
 Skydive Loincloth Alpha {type=number}
 target_sim_space {type=Rotator}
]
