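# Physics asset for the beard of the male medium Historian player body, covering the jaw, beard bones and upper spine.
# A beard capsule hangs off a small master-bone capsule through a joint with wide swing limits and very tight twist.
# The master and jaw bodies each connect to spine_05 with free angular motion.
# spine_05 carries a large, offset capsule that approximates the upper torso.
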PhysicsAsset /BRCosmetics/Characters/Player/Male/Medium/Bodies/M_MED_Historian/Meshes/Parts/M_MED_Historian_dyn
bodies {
 C_jaw [{shapeType=Capsule, length=0.5, radius=0.5, rotation=<rot 90 0 0>}]
 dyn_beard [{shapeType=Capsule, length=3.25, radius=2.6627, center=<1.8289 0 -0.2896>, rotation=<rot 75.0664 180 180>}]
 dyn_beard_MSTR [{shapeType=Capsule, length=0.5, radius=0.5, rotation=<rot 90 0 0>}]
 spine_05 [{shapeType=Capsule, length=22.8125, radius=27.4062, center=<-11.7307 -11.5025 1.9307>, rotation=<rot 0 0 13.378>}]
}
constraints [
 {bone1Name=dyn_beard, bone2Name=dyn_beard_MSTR, swing1LimitDegrees=50, swing2LimitDegrees=50, twistLimitDegrees=5}
 {
  bone1Name=C_jaw
  bone2Name=spine_05
  swing1LimitDegrees=45
  swing1Motion=Free
  swing2LimitDegrees=45
  swing2Motion=Free
  twistLimitDegrees=45
  twistMotion=Free
 }
 {
  bone1Name=dyn_beard_MSTR
  bone2Name=spine_05
  swing1LimitDegrees=45
  swing1Motion=Free
  swing2LimitDegrees=45
  swing2Motion=Free
  twistLimitDegrees=45
  twistMotion=Free
 }
]
